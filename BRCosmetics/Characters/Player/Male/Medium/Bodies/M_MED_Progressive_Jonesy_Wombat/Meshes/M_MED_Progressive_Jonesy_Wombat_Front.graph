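# Physics asset for the Progressive Jonesy Wombat male body cosmetic, simulating the poncho and sleeve dyn_ bones.
# Capsules on pelvis, spine, clavicles and arms form the body shapes, joined by angularly free constraints.
# Box and capsule cloth chains hang from spine_05, both clavicles and upperarm_l with limited swing and twist.
# Extra angularly free constraints cross-link left and right upper back chains, and upperarm_l to its sleeve bones.
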PhysicsAsset /BRCosmetics/Characters/Player/Male/Medium/Bodies/M_MED_Progressive_Jonesy_Wombat/Meshes/M_MED_Progressive_Jonesy_Wombat_Front
bodies {
 clavicle_l [{shapeType=Capsule, length=10.4235, radius=5.5857, center=<7.7216 -0.0001 0>, rotation=<rot 90 0 0>}]
 clavicle_r [{shapeType=Capsule, length=10.4235, radius=5.5857, center=<-8.0004 0 0>, rotation=<rot 90 0 0>}]
 dyn_l_sleeve_top_01 [{shapeType=Box, extentX=8.2767, extentY=19.7795, extentZ=2.4249, center=<4.5578 0 0>}]
 dyn_l_sleeve_top_02 [{shapeType=Box, extentX=8.2767, extentY=19.7795, extentZ=2.4249, center=<4.5578 0 0>}]
 dyn_poncho_back_l_bot_01 [{shapeType=Capsule, length=8.702, radius=1.7907, center=<6.264 0 -0.0001>, rotation=<rot 90 0 0>}]
 dyn_poncho_back_l_bot_02 [{shapeType=Box, extentX=9.273, extentY=15.5307, extentZ=3.1696, center=<5.0098 -0.0001 -2.3748>}]
 dyn_poncho_back_l_bot_03 [{shapeType=Box, extentX=9.273, extentY=15.6745, extentZ=3.1696, center=<5.0098 -0.0001 -1.6894>}]
 dyn_poncho_back_l_bot_04 [{shapeType=Capsule, length=3.822, radius=3.1543, center=<3.0888 -0.3146 -0.2026>, rotation=<rot 0 176.0078 -82.7241>}]
 dyn_poncho_back_l_top_01 [{shapeType=Box, extentX=11.0529, extentY=35.1857, extentZ=2.0556, center=<6.1594 0 1.4818>}]
 dyn_poncho_back_l_top_02 [{shapeType=Box, extentX=8.5851, extentY=36.4179, extentZ=3.2947, center=<4.6805 0 0.0001>}]
 dyn_poncho_back_l_top_03 [{shapeType=Box, extentX=12.7625, extentY=39.7857, extentZ=3.2947, center=<6.8097 0 0.0001>}]
 dyn_poncho_back_r_top_01 [{shapeType=Box, extentX=11.0529, extentY=10.9857, extentZ=1.6878, center=<6.1594 0 1.4818>}]
 dyn_poncho_back_r_top_02 [{shapeType=Box, extentX=8.5851, extentY=10.9857, extentZ=1.0001, center=<4.6805 0 0.0001>}]
 dyn_poncho_back_r_top_03 [{shapeType=Box, extentX=11.3654, extentY=10.9857, extentZ=3.6089, center=<5.4157 0 0.0001>}]
 dyn_poncho_back_r_top_04 [{shapeType=Box, extentX=12.388, extentY=10.9857, extentZ=6.0745, center=<6.048 0 0.0001>}]
 dyn_poncho_front_l_bot_01 [{shapeType=Box, extentX=5, extentY=6.199, extentZ=2.1312, center=<2.2383 0 0>}]
 dyn_poncho_front_l_bot_02 [{shapeType=Box, extentX=8.7458, extentY=7.223, extentZ=0.5, center=<4.6632 0 -0.0002>}]
 dyn_poncho_front_l_bot_03 [{shapeType=Box, extentX=5.3104, extentY=6.2809, extentZ=1.0206, center=<6.8015 -0.0001 -0.0002>}]
 dyn_poncho_front_r_top_01 [{shapeType=Box, extentX=12.052, extentY=14.1562, extentZ=6.6037, center=<5.8075 0 0.0002>}]
 dyn_poncho_front_r_top_02 [{shapeType=Box, extentX=12.052, extentY=14.7171, extentZ=5.6154, center=<5.8075 0 0.0002>}]
 dyn_poncho_front_r_top_03 [{shapeType=Capsule, length=3.4929, radius=3.291, center=<2.9462 0 -0.0001>, rotation=<rot 0 0 90>}]
 dyn_poncho_l_sleeve_back_top_01 [{shapeType=Box, extentX=7.257, extentY=23.7431, extentZ=5.2734, center=<3.6678 0 0>}]
 dyn_poncho_l_sleeve_back_top_02 [{shapeType=Box, extentX=13.337, extentY=26.3081, extentZ=4.6589, center=<6.9368 0 0.0001>}]
 dyn_poncho_l_sleeve_front_top_01 [{shapeType=Box, extentX=8.5882, extentY=11.0574, extentZ=4.2683, center=<4.3647 0 0.1944>}]
 dyn_poncho_l_sleeve_front_top_02 [{shapeType=Box, extentX=12.6386, extentY=12.3257, extentZ=2.7568, center=<6.4559 0 0.627>}]
 dyn_poncho_r_fold_back_01 [{shapeType=Capsule, length=2.2943, radius=2.5659, center=<3.8333 0 -0.0002>, rotation=<rot 90 0 0>}]
 dyn_poncho_r_sleeve_back_01 [{shapeType=Capsule, length=2.5778, radius=2.2893, center=<3.8333 0 -0.0002>, rotation=<rot 90 0 0>}]
 dyn_poncho_r_sleeve_back_base [{shapeType=Capsule, length=2.1109, radius=1.8891, center=<3.8333 0 -0.0002>, rotation=<rot 90 0 0>}]
 dyn_poncho_r_sleeve_front_01 [{shapeType=Capsule, length=3.6664, radius=2.5293, center=<4.1319 -0.0001 -0.0002>, rotation=<rot 90 0 0>}]
 dyn_poncho_r_sleeve_front_02 [{shapeType=Capsule, length=5.3213, radius=2.2438, center=<5.0745 0 0>, rotation=<rot 90 0 0>}]
 lowerarm_l [{shapeType=Capsule, length=16.0267, radius=5.9725, center=<10.6892 0.0003 0>, rotation=<rot 90 0 0>}]
 pelvis [{shapeType=Capsule, length=16.7417, radius=14.0765}, {shapeType=Capsule, length=16.7417, radius=14.0765, center=<0.2064 -3.1347 0>}]
 spine_02 [{shapeType=Capsule, length=13.4522, radius=11.3214}, {shapeType=Capsule, length=20.313, radius=11.3214, center=<0.6512 -3.4586 0.0001>}]
 spine_04 [{shapeType=Capsule, length=15.4336, radius=12.0789}]
 spine_05 [{shapeType=Capsule, length=16.6046, radius=12.0789}]
 upperarm_l [{shapeType=Capsule, length=13.9241, radius=5.5417, center=<13.3607 0.0002 0>, rotation=<rot 90 0 0>}]
 upperarm_r [{shapeType=Capsule, length=13.9241, radius=5.5417, center=<-11.7496 -0.0001 0>, rotation=<rot 90 0 0>}]
}
constraints [
 {
  bone1Name=spine_05
  bone2Name=spine_04
  swing1LimitDegrees=45
  swing1Motion=Free
  swing2LimitDegrees=45
  swing2Motion=Free
  twistLimitDegrees=45
  twistMotion=Free
 }
 {
  bone1Name=clavicle_l
  bone2Name=spine_05
  swing1LimitDegrees=45
  swing1Motion=Free
  swing2LimitDegrees=45
  swing2Motion=Free
  twistLimitDegrees=45
  twistMotion=Free
 }
 {
  bone1Name=upperarm_l
  bone2Name=clavicle_l
  swing1LimitDegrees=45
  swing1Motion=Free
  swing2LimitDegrees=45
  swing2Motion=Free
  twistLimitDegrees=45
  twistMotion=Free
 }
 {
  bone1Name=upperarm_r
  bone2Name=clavicle_r
  swing1LimitDegrees=45
  swing1Motion=Free
  swing2LimitDegrees=45
  swing2Motion=Free
  twistLimitDegrees=45
  twistMotion=Free
 }
 {
  bone1Name=clavicle_r
  bone2Name=spine_05
  swing1LimitDegrees=45
  swing1Motion=Free
  swing2LimitDegrees=45
  swing2Motion=Free
  twistLimitDegrees=45
  twistMotion=Free
 }
 {bone1Name=dyn_poncho_back_l_top_03, bone2Name=dyn_poncho_back_l_top_02, swing1LimitDegrees=15, swing2LimitDegrees=30, twistLimitDegrees=15}
 {bone1Name=dyn_poncho_back_l_top_01, bone2Name=spine_05, swing1LimitDegrees=15, swing2LimitDegrees=30, twistLimitDegrees=15}
 {bone1Name=dyn_poncho_back_l_top_02, bone2Name=dyn_poncho_back_l_top_01, swing1LimitDegrees=15, swing2LimitDegrees=30, twistLimitDegrees=15}
 {bone1Name=dyn_poncho_back_r_top_04, bone2Name=dyn_poncho_back_r_top_03, swing1LimitDegrees=30, swing2LimitDegrees=30, twistLimitDegrees=15}
 {bone1Name=dyn_poncho_back_r_top_01, bone2Name=spine_05, swing1LimitDegrees=30, swing2LimitDegrees=30, twistLimitDegrees=15}
 {bone1Name=dyn_poncho_back_r_top_02, bone2Name=dyn_poncho_back_r_top_01, swing1LimitDegrees=30, swing2LimitDegrees=30, twistLimitDegrees=15}
 {bone1Name=dyn_poncho_back_r_top_03, bone2Name=dyn_poncho_back_r_top_02, swing1LimitDegrees=30, swing2LimitDegrees=30, twistLimitDegrees=15}
 {
  bone1Name=dyn_poncho_r_fold_back_01
  bone2Name=dyn_poncho_r_sleeve_back_base
  swing1LimitDegrees=5
  swing2LimitDegrees=5
  twistLimitDegrees=5
 }
 {
  bone1Name=dyn_poncho_r_sleeve_back_01
  bone2Name=dyn_poncho_r_sleeve_back_base
  swing1LimitDegrees=20
  swing2LimitDegrees=20
  twistLimitDegrees=10
 }
 {bone1Name=dyn_poncho_r_sleeve_back_base, bone2Name=clavicle_r, swing1LimitDegrees=10, swing2LimitDegrees=10, twistLimitDegrees=5}
 {bone1Name=dyn_poncho_r_sleeve_front_01, bone2Name=clavicle_r, swing1LimitDegrees=5, swing2LimitDegrees=5, twistLimitDegrees=5}
 {
  bone1Name=dyn_poncho_r_sleeve_front_02
  bone2Name=dyn_poncho_r_sleeve_front_01
  swing1LimitDegrees=30
  swing2LimitDegrees=30
  twistLimitDegrees=20
 }
 {bone1Name=dyn_poncho_front_r_top_03, bone2Name=dyn_poncho_front_r_top_02, swing1LimitDegrees=30, swing2LimitDegrees=30, twistLimitDegrees=20}
 {bone1Name=dyn_poncho_front_r_top_01, bone2Name=clavicle_l, swing1LimitDegrees=30, swing2LimitDegrees=30, twistLimitDegrees=30}
 {bone1Name=dyn_poncho_front_r_top_02, bone2Name=dyn_poncho_front_r_top_01, swing1LimitDegrees=30, swing2LimitDegrees=30, twistLimitDegrees=30}
 {
  bone1Name=dyn_poncho_l_sleeve_front_top_02
  bone2Name=dyn_poncho_l_sleeve_front_top_01
  swing1LimitDegrees=30
  swing2LimitDegrees=30
  twistLimitDegrees=5
 }
 {bone1Name=dyn_poncho_l_sleeve_front_top_01, bone2Name=clavicle_l, swing1LimitDegrees=15, swing2LimitDegrees=15, twistLimitDegrees=5}
 {bone1Name=dyn_poncho_front_l_bot_01, bone2Name=clavicle_l, swing1LimitDegrees=5, swing2LimitDegrees=5, twistLimitDegrees=5}
 {bone1Name=dyn_poncho_front_l_bot_02, bone2Name=dyn_poncho_front_l_bot_01, swing1LimitDegrees=20, swing2LimitDegrees=20, twistLimitDegrees=15}
 {bone1Name=dyn_poncho_front_l_bot_03, bone2Name=dyn_poncho_front_l_bot_02, swing1LimitDegrees=20, swing2LimitDegrees=20, twistLimitDegrees=20}
 {bone1Name=dyn_poncho_back_l_bot_04, bone2Name=dyn_poncho_back_l_bot_03, swing1LimitDegrees=15, swing2LimitDegrees=15, twistLimitDegrees=7.5}
 {bone1Name=dyn_poncho_back_l_bot_01, bone2Name=clavicle_l, swing1LimitDegrees=45, swing2LimitDegrees=20, twistLimitDegrees=25}
 {bone1Name=dyn_poncho_back_l_bot_02, bone2Name=dyn_poncho_back_l_bot_01, swing1LimitDegrees=40, swing2LimitDegrees=40, twistLimitDegrees=40}
 {bone1Name=dyn_poncho_back_l_bot_03, bone2Name=dyn_poncho_back_l_bot_02, swing1LimitDegrees=20, swing2LimitDegrees=20, twistLimitDegrees=7.5}
 {
  bone1Name=dyn_poncho_l_sleeve_back_top_02
  bone2Name=dyn_poncho_l_sleeve_back_top_01
  swing1LimitDegrees=15
  swing2LimitDegrees=15
  twistLimitDegrees=10
 }
 {bone1Name=dyn_poncho_l_sleeve_back_top_01, bone2Name=clavicle_l, swing1LimitDegrees=15, swing2LimitDegrees=15, twistLimitDegrees=10}
 {
  bone1Name=spine_04
  bone2Name=spine_02
  swing1LimitDegrees=45
  swing1Motion=Free
  swing2LimitDegrees=45
  swing2Motion=Free
  twistLimitDegrees=45
  twistMotion=Free
 }
 {
  bone1Name=spine_02
  bone2Name=pelvis
  swing1LimitDegrees=45
  swing1Motion=Free
  swing2LimitDegrees=45
  swing2Motion=Free
  twistLimitDegrees=45
  twistMotion=Free
 }
 {
  bone1Name=dyn_poncho_back_l_top_01
  bone2Name=dyn_poncho_back_r_top_01
  swing1LimitDegrees=45
  swing1Motion=Free
  swing2LimitDegrees=45
  swing2Motion=Free
  twistLimitDegrees=45
  twistMotion=Free
 }
 {
  bone1Name=dyn_poncho_back_l_top_02
  bone2Name=dyn_poncho_back_r_top_02
  swing1LimitDegrees=45
  swing1Motion=Free
  swing2LimitDegrees=45
  swing2Motion=Free
  twistLimitDegrees=45
  twistMotion=Free
 }
 {
  bone1Name=dyn_poncho_back_l_top_03
  bone2Name=dyn_poncho_back_r_top_03
  swing1LimitDegrees=45
  swing1Motion=Free
  swing2LimitDegrees=45
  swing2Motion=Free
  twistLimitDegrees=45
  twistMotion=Free
 }
 {bone1Name=dyn_l_sleeve_top_01, bone2Name=upperarm_l, swing1LimitDegrees=15, swing2LimitDegrees=15, twistLimitDegrees=15}
 {bone1Name=dyn_l_sleeve_top_02, bone2Name=dyn_l_sleeve_top_01, swing1LimitDegrees=15, swing2LimitDegrees=15, twistLimitDegrees=15}
 {
  bone1Name=upperarm_l
  bone2Name=dyn_l_sleeve_top_02
  swing1LimitDegrees=45
  swing1Motion=Free
  swing2LimitDegrees=45
  swing2Motion=Free
  twistLimitDegrees=45
  twistMotion=Free
 }
 {
  bone1Name=upperarm_l
  bone2Name=dyn_l_sleeve_top_01
  swing1LimitDegrees=45
  swing1Motion=Free
  swing2LimitDegrees=45
  swing2Motion=Free
  twistLimitDegrees=45
  twistMotion=Free
 }
 {
  bone1Name=lowerarm_l
  bone2Name=upperarm_l
  swing1LimitDegrees=45
  swing1Motion=Free
  swing2LimitDegrees=45
  swing2Motion=Free
  twistLimitDegrees=45
  twistMotion=Free
 }
]
