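# Physics asset for the Progressive Jonesy Wombat male body cosmetic, set up for its dynamic poncho and sleeve bones.
# Capsules cover pelvis, spine, clavicles, upper arms and left forearm, joined by angularly free constraints.
# Capsule and box cloth chains hang from upper spine, clavicles and left upper arm with limited swing and twist.
# Extra angularly free constraints cross-link left and right upper back poncho chains at matching segments.
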PhysicsAsset /BRCosmetics/Characters/Player/Male/Medium/Bodies/M_MED_Progressive_Jonesy_Wombat/Meshes/M_MED_Progressive_Jonesy_Wombat_Physics
bodies {
 clavicle_l [{shapeType=Capsule, length=10.4235, radius=5.5857, center=<7.7216 -0.0001 0>, rotation=<rot 90 0 0>}]
 clavicle_r [{shapeType=Capsule, length=10.4235, radius=5.5857, center=<-8.0004 0 0>, rotation=<rot 90 0 0>}]
 dyn_l_sleeve_top_01 [{shapeType=Capsule, length=12.25, radius=3.75, shapeName=dyn_l_sleeve_top_01_capsule, center=<4.4569 0 0>, rotation=<rot 69.9999 -89.9999 0>}]
 dyn_l_sleeve_top_02 [{shapeType=Capsule, length=10.25, radius=4, shapeName=dyn_l_sleeve_top_02_capsule, center=<4.1206 0 0>, rotation=<rot 69.9997 -89.9998 0>}]
 dyn_poncho_back_l_bot_01 [{shapeType=Capsule, length=8.702, radius=1.7907, center=<6.264 0 -0.0001>, rotation=<rot 90 0 0>}]
 dyn_poncho_back_l_bot_02 [{
   shapeType=Capsule
   length=11.25
   radius=3.755
   shapeName=dyn_poncho_back_l_bot_02_capsule
   center=<3.1545 -0.2188 -0.4054>
   rotation=<rot -90 172.875 97.1251>
  }]
 dyn_poncho_back_l_bot_03 [{
   shapeType=Capsule
   length=12.5684
   radius=3.9407
   shapeName=dyn_poncho_back_l_bot_03_capsule
   center=<4.6863 -0.8224 0.4501>
   rotation=<rot 2.3097 -178.6733 -68.6659>
  }]
 dyn_poncho_back_l_bot_04 [{shapeType=Capsule, length=3.822, radius=3.1543, center=<3.0888 -0.3146 -0.2026>, rotation=<rot 0 176.0078 -82.7241>}]
 dyn_poncho_back_l_top_01 [{shapeType=Box, extentX=11.0529, extentY=35.1857, extentZ=2.0556, center=<6.1594 0 1.4818>}]
 dyn_poncho_back_l_top_02 [{shapeType=Box, extentX=8.5851, extentY=36.4179, extentZ=3.2947, center=<4.6805 0 0.0001>}]
 dyn_poncho_back_l_top_03 [{shapeType=Box, extentX=12.7625, extentY=39.7857, extentZ=3.2947, center=<6.8097 0 0.0001>}]
 dyn_poncho_back_r_top_01 [{
   shapeType=Capsule
   length=10.7064
   radius=3.7762
   shapeName=dyn_poncho_back_r_top_01_capsule
   center=<7.5868 0.4658 1.9456>
   rotation=<rot 0.4633 0.4715 -89.6111>
  }]
 dyn_poncho_back_r_top_02 [{
   shapeType=Capsule
   length=6.485
   radius=4.6598
   shapeName=dyn_poncho_back_r_top_02_capsule
   center=<4.0164 -0.5709 1.7346>
   rotation=<rot -0.7487 -180 -90>
  }]
 dyn_poncho_back_r_top_03 [{
   shapeType=Capsule
   length=8.1969
   radius=4.2101
   shapeName=dyn_poncho_back_r_top_03_capsule
   center=<3.7643 0.6787 0.1191>
   rotation=<rot -10.4318 -180 -90>
  }]
 dyn_poncho_back_r_top_04 [{
   shapeType=Capsule
   length=7.5453
   radius=7.0128
   shapeName=dyn_poncho_back_r_top_04_capsule
   center=<5.3656 0.8194 -0.1824>
   rotation=<rot -3.6783 180 -90>
  }]
 dyn_poncho_front_l_bot_01 [{shapeType=Box, extentX=5, extentY=6.199, extentZ=2.1312, center=<2.2383 0 0>}]
 dyn_poncho_front_l_bot_02 [{shapeType=Box, extentX=8.7458, extentY=7.223, extentZ=0.5, center=<4.6632 0 -0.0002>}]
 dyn_poncho_front_l_bot_03 [{shapeType=Box, extentX=8.7458, extentY=7.223, extentZ=2.4843, center=<4.6632 0 -0.0002>}]
 dyn_poncho_front_r_top_01 [{
   shapeType=Capsule
   length=9.0898
   radius=5.872
   shapeName=dyn_poncho_front_r_top_01_capsule
   center=<5.9044 -1.3798 -1.9662>
   rotation=<rot -2.969 176.3683 -87.5058>
  }]
 dyn_poncho_front_r_top_02 [{
   shapeType=Capsule
   length=10.8474
   radius=4.8217
   shapeName=dyn_poncho_front_r_top_02_capsule
   center=<6.7881 -2.1325 -0.6947>
   rotation=<rot -15.9534 -178.0733 -76.1213>
  }]
 dyn_poncho_front_r_top_03 [{shapeType=Capsule, length=3.4929, radius=3.291, center=<2.9462 0 -0.0001>, rotation=<rot 0 0 90>}]
 dyn_poncho_l_sleeve_back_top_01 [{shapeType=Box, extentX=7.257, extentY=23.7431, extentZ=5.2734, center=<3.6678 0 0>}]
 dyn_poncho_l_sleeve_back_top_02 [{shapeType=Box, extentX=13.337, extentY=26.3081, extentZ=4.6589, center=<6.9368 0 0.0001>}]
 dyn_poncho_l_sleeve_front_top_01 [{
   shapeType=Capsule
   length=9
   radius=3.755
   shapeName=dyn_poncho_l_sleeve_front_top_01_capsule
   center=<4.1065 0 0>
   rotation=<rot 80 -90.0004 -0.0004>
  }]
 dyn_poncho_l_sleeve_front_top_02 [{
   shapeType=Capsule
   length=8.255
   radius=4.255
   shapeName=dyn_poncho_l_sleeve_front_top_02_capsule
   center=<10.0634 -0.0801 0.4548>
   rotation=<rot 75.8941 -44.5618 134.5606>
  }]
 dyn_poncho_r_fold_back_01 [{shapeType=Capsule, length=2.2943, radius=2.5659, center=<3.8333 0 -0.0002>, rotation=<rot 90 0 0>}]
 dyn_poncho_r_sleeve_back_01 [{shapeType=Capsule, length=6.3278, radius=3.0393, center=<3.6743 -0.9017 -0.0002>, rotation=<rot 90 -165.9638 -65.9637>}]
 dyn_poncho_r_sleeve_back_base [{shapeType=Capsule, length=2.1109, radius=1.8891, center=<3.8333 0 -0.0002>, rotation=<rot 90 0 0>}]
 dyn_poncho_r_sleeve_front_01 [{shapeType=Capsule, length=3.6664, radius=2.5293, center=<4.1319 -0.0001 -0.0002>, rotation=<rot 90 0 0>}]
 dyn_poncho_r_sleeve_front_02 [{shapeType=Capsule, length=5.3213, radius=2.2438, center=<5.0745 0 0>, rotation=<rot 90 0 0>}]
 lowerarm_l [{shapeType=Capsule, length=15.0267, radius=6.9725, center=<10.6892 0.0003 0>, rotation=<rot 90 0 0>}]
 pelvis [{shapeType=Capsule, length=16.7417, radius=14.0765}, {shapeType=Capsule, length=16.7417, radius=14.0765, center=<0.2064 -3.1347 0>}]
 spine_02 [{shapeType=Capsule, length=13.4522, radius=11.3214}, {shapeType=Capsule, length=20.313, radius=11.3214, center=<0.6512 -3.4586 0.0001>}]
 spine_04 [{shapeType=Capsule, length=15.4336, radius=12.0789}]
 spine_05 [{shapeType=Capsule, length=16.6046, radius=12.0789}]
 upperarm_l [{shapeType=Capsule, length=13.9241, radius=5.5417, center=<13.3607 0.0002 0>, rotation=<rot 90 0 0>}]
 upperarm_r [{shapeType=Capsule, length=13.9241, radius=5.5417, center=<-11.7496 -0.0001 0>, rotation=<rot 90 0 0>}]
}
constraints [
 {
  bone1Name=spine_05
  bone2Name=spine_04
  swing1LimitDegrees=45
  swing1Motion=Free
  swing2LimitDegrees=45
  swing2Motion=Free
  twistLimitDegrees=45
  twistMotion=Free
 }
 {
  bone1Name=clavicle_l
  bone2Name=spine_05
  swing1LimitDegrees=45
  swing1Motion=Free
  swing2LimitDegrees=45
  swing2Motion=Free
  twistLimitDegrees=45
  twistMotion=Free
 }
 {
  bone1Name=upperarm_l
  bone2Name=clavicle_l
  swing1LimitDegrees=45
  swing1Motion=Free
  swing2LimitDegrees=45
  swing2Motion=Free
  twistLimitDegrees=45
  twistMotion=Free
 }
 {
  bone1Name=upperarm_r
  bone2Name=clavicle_r
  swing1LimitDegrees=45
  swing1Motion=Free
  swing2LimitDegrees=45
  swing2Motion=Free
  twistLimitDegrees=45
  twistMotion=Free
 }
 {
  bone1Name=clavicle_r
  bone2Name=spine_05
  swing1LimitDegrees=45
  swing1Motion=Free
  swing2LimitDegrees=45
  swing2Motion=Free
  twistLimitDegrees=45
  twistMotion=Free
 }
 {bone1Name=dyn_poncho_back_l_top_03, bone2Name=dyn_poncho_back_l_top_02, swing1LimitDegrees=15, swing2LimitDegrees=30, twistLimitDegrees=15}
 {bone1Name=dyn_poncho_back_l_top_01, bone2Name=spine_05, swing1LimitDegrees=15, swing2LimitDegrees=30, twistLimitDegrees=15}
 {bone1Name=dyn_poncho_back_l_top_02, bone2Name=dyn_poncho_back_l_top_01, swing1LimitDegrees=15, swing2LimitDegrees=30, twistLimitDegrees=15}
 {bone1Name=dyn_poncho_back_r_top_04, bone2Name=dyn_poncho_back_r_top_03, swing1LimitDegrees=30, swing2LimitDegrees=30, twistLimitDegrees=15}
 {bone1Name=dyn_poncho_back_r_top_01, bone2Name=spine_05, swing1LimitDegrees=30, swing2LimitDegrees=30, twistLimitDegrees=15}
 {bone1Name=dyn_poncho_back_r_top_02, bone2Name=dyn_poncho_back_r_top_01, swing1LimitDegrees=30, swing2LimitDegrees=30, twistLimitDegrees=15}
 {bone1Name=dyn_poncho_back_r_top_03, bone2Name=dyn_poncho_back_r_top_02, swing1LimitDegrees=30, swing2LimitDegrees=30, twistLimitDegrees=15}
 {
  bone1Name=dyn_poncho_r_fold_back_01
  bone2Name=dyn_poncho_r_sleeve_back_base
  swing1LimitDegrees=5
  swing2LimitDegrees=5
  twistLimitDegrees=5
 }
 {
  bone1Name=dyn_poncho_r_sleeve_back_01
  bone2Name=dyn_poncho_r_sleeve_back_base
  swing1LimitDegrees=20
  swing2LimitDegrees=20
  twistLimitDegrees=10
 }
 {bone1Name=dyn_poncho_r_sleeve_back_base, bone2Name=clavicle_r, swing1LimitDegrees=10, swing2LimitDegrees=10, twistLimitDegrees=5}
 {bone1Name=dyn_poncho_r_sleeve_front_01, bone2Name=clavicle_r, swing1LimitDegrees=5, swing2LimitDegrees=5, twistLimitDegrees=5}
 {
  bone1Name=dyn_poncho_r_sleeve_front_02
  bone2Name=dyn_poncho_r_sleeve_front_01
  swing1LimitDegrees=30
  swing2LimitDegrees=30
  twistLimitDegrees=20
 }
 {bone1Name=dyn_poncho_front_r_top_03, bone2Name=dyn_poncho_front_r_top_02, swing1LimitDegrees=30, swing2LimitDegrees=30, twistLimitDegrees=20}
 {bone1Name=dyn_poncho_front_r_top_01, bone2Name=clavicle_l, swing1LimitDegrees=30, swing2LimitDegrees=30, twistLimitDegrees=30}
 {bone1Name=dyn_poncho_front_r_top_02, bone2Name=dyn_poncho_front_r_top_01, swing1LimitDegrees=30, swing2LimitDegrees=30, twistLimitDegrees=30}
 {
  bone1Name=dyn_poncho_l_sleeve_front_top_02
  bone2Name=dyn_poncho_l_sleeve_front_top_01
  swing1LimitDegrees=30
  swing2LimitDegrees=30
  twistLimitDegrees=5
 }
 {bone1Name=dyn_poncho_l_sleeve_front_top_01, bone2Name=clavicle_l, swing1LimitDegrees=15, swing2LimitDegrees=15, twistLimitDegrees=5}
 {bone1Name=dyn_poncho_front_l_bot_01, bone2Name=clavicle_l, swing1LimitDegrees=5, swing2LimitDegrees=5, twistLimitDegrees=5}
 {bone1Name=dyn_poncho_front_l_bot_02, bone2Name=dyn_poncho_front_l_bot_01, swing1LimitDegrees=20, swing2LimitDegrees=20, twistLimitDegrees=15}
 {bone1Name=dyn_poncho_front_l_bot_03, bone2Name=dyn_poncho_front_l_bot_02, swing1LimitDegrees=20, swing2LimitDegrees=20, twistLimitDegrees=15}
 {bone1Name=dyn_poncho_back_l_bot_04, bone2Name=dyn_poncho_back_l_bot_03, swing1LimitDegrees=15, swing2LimitDegrees=15, twistLimitDegrees=7.5}
 {bone1Name=dyn_poncho_back_l_bot_01, bone2Name=clavicle_l, swing1LimitDegrees=45, swing2LimitDegrees=20, twistLimitDegrees=25}
 {bone1Name=dyn_poncho_back_l_bot_02, bone2Name=dyn_poncho_back_l_bot_01, swing1LimitDegrees=40, swing2LimitDegrees=40, twistLimitDegrees=40}
 {bone1Name=dyn_poncho_back_l_bot_03, bone2Name=dyn_poncho_back_l_bot_02, swing1LimitDegrees=20, swing2LimitDegrees=20, twistLimitDegrees=7.5}
 {
  bone1Name=dyn_poncho_l_sleeve_back_top_02
  bone2Name=dyn_poncho_l_sleeve_back_top_01
  swing1LimitDegrees=15
  swing2LimitDegrees=15
  twistLimitDegrees=10
 }
 {bone1Name=dyn_poncho_l_sleeve_back_top_01, bone2Name=clavicle_l, swing1LimitDegrees=15, swing2LimitDegrees=15, twistLimitDegrees=10}
 {
  bone1Name=spine_04
  bone2Name=spine_02
  swing1LimitDegrees=45
  swing1Motion=Free
  swing2LimitDegrees=45
  swing2Motion=Free
  twistLimitDegrees=45
  twistMotion=Free
 }
 {
  bone1Name=spine_02
  bone2Name=pelvis
  swing1LimitDegrees=45
  swing1Motion=Free
  swing2LimitDegrees=45
  swing2Motion=Free
  twistLimitDegrees=45
  twistMotion=Free
 }
 {
  bone1Name=dyn_poncho_back_l_top_01
  bone2Name=dyn_poncho_back_r_top_01
  swing1LimitDegrees=45
  swing1Motion=Free
  swing2LimitDegrees=45
  swing2Motion=Free
  twistLimitDegrees=45
  twistMotion=Free
 }
 {
  bone1Name=dyn_poncho_back_l_top_02
  bone2Name=dyn_poncho_back_r_top_02
  swing1LimitDegrees=45
  swing1Motion=Free
  swing2LimitDegrees=45
  swing2Motion=Free
  twistLimitDegrees=45
  twistMotion=Free
 }
 {
  bone1Name=dyn_poncho_back_l_top_03
  bone2Name=dyn_poncho_back_r_top_03
  swing1LimitDegrees=45
  swing1Motion=Free
  swing2LimitDegrees=45
  swing2Motion=Free
  twistLimitDegrees=45
  twistMotion=Free
 }
 {bone1Name=dyn_l_sleeve_top_01, bone2Name=upperarm_l, swing1LimitDegrees=15, swing2LimitDegrees=15, twistLimitDegrees=15}
 {bone1Name=dyn_l_sleeve_top_02, bone2Name=dyn_l_sleeve_top_01, swing1LimitDegrees=15, swing2LimitDegrees=15, twistLimitDegrees=15}
 {
  bone1Name=upperarm_l
  bone2Name=dyn_l_sleeve_top_02
  swing1LimitDegrees=45
  swing1Motion=Free
  swing2LimitDegrees=45
  swing2Motion=Free
  twistLimitDegrees=45
  twistMotion=Free
 }
 {
  bone1Name=upperarm_l
  bone2Name=dyn_l_sleeve_top_01
  swing1LimitDegrees=45
  swing1Motion=Free
  swing2LimitDegrees=45
  swing2Motion=Free
  twistLimitDegrees=45
  twistMotion=Free
 }
 {
  bone1Name=lowerarm_l
  bone2Name=upperarm_l
  swing1LimitDegrees=45
  swing1Motion=Free
  swing2LimitDegrees=45
  swing2Motion=Free
  twistLimitDegrees=45
  twistMotion=Free
 }
]
