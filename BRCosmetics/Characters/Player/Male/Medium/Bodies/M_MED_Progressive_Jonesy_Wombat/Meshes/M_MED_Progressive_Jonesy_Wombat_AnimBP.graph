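# Animation Blueprint for the Progressive Jonesy Wombat cosmetic body, adding secondary motion to its poncho, pouches, crossbow and shirt.
# It copies the parent mesh pose, applies weighted constraints and bone offsets, runs two Kepler-asset RigidBody sims, then blends front-end or in-game Wombat sims from a cached pose.
# Post-physics corrections follow alphas the EventGraph derives from arm angles, weapon type, skydiving and backbling.
# Delayed events reset dynamics.
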AnimBlueprint /BRCosmetics/Characters/Player/Male/Medium/Bodies/M_MED_Progressive_Jonesy_Wombat/Meshes/M_MED_Progressive_Jonesy_Wombat_AnimBP
target_skeleton=&/M_MED_Progressive_Jonesy_Wombat_Skeleton
graphs [
 AnimGraph {nodes [
   n1 = "Constraint - dyn_shoulder_pad_MSTR (PPPPPPPP)"!(
    ComponentPose
    Alpha=1
    ConstraintWeights_0=0.6003
    ConstraintWeights_1:float=0.0421
    ConstraintWeights_2:float=0.0064
    ConstraintWeights_3:float=0.1164
    ConstraintWeights_4:float=0.2186
    ConstraintWeights_5:float=0.0125
    ConstraintWeights_6:float=0.0036
    ConstraintWeights_7:float=0.0001
   ) -> (Pose) {
    alphaBoolBlend {blendOption=Linear}
    boneToModify {boneName=dyn_shoulder_pad_MSTR}
    constraintSetup [
     {targetBone {boneName=clavicle_l}}
     {targetBone {boneName=spine_03}}
     {targetBone {boneName=spine_04}}
     {targetBone {boneName=spine_05}}
     {targetBone {boneName=upperarm_twist_01_l}}
     {targetBone {boneName=neck_02}}
     {targetBone {boneName=neck_01}}
     {targetBone {boneName=upperarm_l}}
    ]
    constraintWeights [1,1,1,1,1,1,1,1]
   }
   "Constraint - dyn_l_shoulder_MSTR (PPPPPPPP)"!(
    ComponentPose=@n1
    Alpha=1
    ConstraintWeights_0=0.6003
    ConstraintWeights_1:float=0.0421
    ConstraintWeights_2:float=0.0064
    ConstraintWeights_3:float=0.1164
    ConstraintWeights_4:float=0.2186
    ConstraintWeights_5:float=0.0125
    ConstraintWeights_6:float=0.0036
    ConstraintWeights_7:float=0.0001
   ) -> (Pose) {
    alphaBoolBlend {blendOption=Linear}
    boneToModify {boneName=dyn_l_shoulder_MSTR}
    constraintSetup [
     {targetBone {boneName=clavicle_l}}
     {targetBone {boneName=spine_03}}
     {targetBone {boneName=spine_04}}
     {targetBone {boneName=spine_05}}
     {targetBone {boneName=upperarm_twist_01_l}}
     {targetBone {boneName=neck_02}}
     {targetBone {boneName=neck_01}}
     {targetBone {boneName=upperarm_l}}
    ]
    constraintWeights [1,1,1,1,1,1,1,1]
   }
   n2 = "Transform (Modify) Bone - Bone: dyn_shoulderpad_MSTR"(Alpha=1, Translation="0,0,-0.5", Rotation="0,-3,10", Scale="1,1,1") {
    rotationMode=BMM_Additive
    rotationSpace=BCS_BoneSpace
    translationMode=BMM_Additive
    translationSpace=BCS_BoneSpace
    alphaBoolBlend {blendOption=Linear}
    alphaScaleBiasClamp {interpSpeedIncreasing=40}
    boneToModify {boneName=dyn_shoulderpad_MSTR}
    rotation {roll=10, yaw=-3}
    translation {z=-0.5}
   }
   n3 = "Transform (Modify) Bone - Bone: dyn_shoulder_pad"(ComponentPose=@n2, Alpha=1, Translation="0,0,0.5", Rotation="0,0,0", Scale="1,1,1") {
    rotationMode=BMM_Additive
    rotationSpace=BCS_BoneSpace
    translationMode=BMM_Additive
    translationSpace=BCS_BoneSpace
    alphaBoolBlend {blendOption=Linear}
    alphaScaleBiasClamp {interpSpeedIncreasing=40}
    boneToModify {boneName=dyn_shoulder_pad}
    translation {z=0.5}
   }
   n4 = "Transform (Modify) Bone - Bone: dyn_shoulderpad_MSTR"!(ComponentPose=@n3, bAlphaBoolEnabled:bool=Get bIsTargeting(), Translation="0,0,0", Rotation="0,0,-20", Scale="1,1,1") -> (Pose) {
    alphaInputType=Bool
    rotationMode=BMM_Additive
    rotationSpace=BCS_BoneSpace
    scaleMode=BMM_Additive
    scaleSpace=BCS_BoneSpace
    translationMode=BMM_Additive
    translationSpace=BCS_BoneSpace
    alphaBoolBlend {blendInTime=0.1, blendOption=Linear, blendOutTime=0.1}
    boneToModify {boneName=dyn_shoulderpad_MSTR}
    rotation {roll=-20}
   }
   n5 = "Constraint - dyn_pouch_rt_MSTR_01 (PPPPPP)"!(
    ComponentPose=@n4
    Alpha=1
    ConstraintWeights_0=0.3474
    ConstraintWeights_1:float=0.2042
    ConstraintWeights_2:float=0.1286
    ConstraintWeights_3:float=0.0288
    ConstraintWeights_4:float=0.0003
    ConstraintWeights_5:float=0.2907
   ) -> (Pose) {
    alphaBoolBlend {blendOption=Linear}
    boneToModify {boneName=dyn_pouch_rt_MSTR_01}
    constraintSetup [
     {targetBone {boneName=spine_01}}
     {targetBone {boneName=spine_02}}
     {targetBone {boneName=spine_03}}
     {targetBone {boneName=spine_04}}
     {targetBone {boneName=spine_05}}
     {targetBone {boneName=pelvis}}
    ]
    constraintWeights [1,1,1,1,1,1]
   }
   n6 = "Constraint - dyn_pouch_rt_MSTR_02 (PPPPPP)"!(
    ComponentPose=@n5
    Alpha=1
    ConstraintWeights_0=0.4033
    ConstraintWeights_1:float=0.1874
    ConstraintWeights_2:float=0.0916
    ConstraintWeights_3:float=0.0299
    ConstraintWeights_4:float=0.0002
    ConstraintWeights_5:float=0.2875
   ) -> (Pose) {
    alphaBoolBlend {blendOption=Linear}
    boneToModify {boneName=dyn_pouch_rt_MSTR_02}
    constraintSetup [
     {targetBone {boneName=spine_01}}
     {targetBone {boneName=spine_02}}
     {targetBone {boneName=spine_03}}
     {targetBone {boneName=spine_04}}
     {targetBone {boneName=spine_05}}
     {targetBone {boneName=pelvis}}
    ]
    constraintWeights [1,1,1,1,1,1]
   }
   n7 = "Constraint - dyn_pouch_rt_MSTR_03 (PPPPPP)"!(
    ComponentPose=@n6
    Alpha=1
    ConstraintWeights_0=0.4115
    ConstraintWeights_1:float=0.1997
    ConstraintWeights_2:float=0.0909
    ConstraintWeights_3:float=0.0467
    ConstraintWeights_4:float=0.0002
    ConstraintWeights_5:float=0.2511
   ) -> (Pose) {
    alphaBoolBlend {blendOption=Linear}
    boneToModify {boneName=dyn_pouch_rt_MSTR_03}
    constraintSetup [
     {targetBone {boneName=spine_01}}
     {targetBone {boneName=spine_02}}
     {targetBone {boneName=spine_03}}
     {targetBone {boneName=spine_04}}
     {targetBone {boneName=spine_05}}
     {targetBone {boneName=pelvis}}
    ]
    constraintWeights [1,1,1,1,1,1]
   }
   n8 = "Constraint - dyn_pouch_bk_MSTR_01 (PPPPPP)"!(
    ComponentPose=@n7
    Alpha=1
    ConstraintWeights_0=0.3609
    ConstraintWeights_1:float=0.2567
    ConstraintWeights_2:float=0.151
    ConstraintWeights_3:float=0.0746
    ConstraintWeights_4:float=0.0017
    ConstraintWeights_5:float=0.1551
   ) -> (Pose) {
    alphaBoolBlend {blendOption=Linear}
    boneToModify {boneName=dyn_pouch_bk_MSTR_01}
    constraintSetup [
     {targetBone {boneName=spine_01}}
     {targetBone {boneName=spine_02}}
     {targetBone {boneName=spine_03}}
     {targetBone {boneName=spine_04}}
     {targetBone {boneName=spine_05}}
     {targetBone {boneName=pelvis}}
    ]
    constraintWeights [1,1,1,1,1,1]
   }
   n9 = "Constraint - dyn_pouch_bk_MSTR_02 (PPPPPP)"!(
    ComponentPose=@n8
    Alpha=1
    ConstraintWeights_0=0.3562
    ConstraintWeights_1:float=0.2572
    ConstraintWeights_2:float=0.1561
    ConstraintWeights_3:float=0.0804
    ConstraintWeights_4:float=0.0024
    ConstraintWeights_5:float=0.1477
   ) -> (Pose) {
    alphaBoolBlend {blendOption=Linear}
    boneToModify {boneName=dyn_pouch_bk_MSTR_02}
    constraintSetup [
     {targetBone {boneName=spine_01}}
     {targetBone {boneName=spine_02}}
     {targetBone {boneName=spine_03}}
     {targetBone {boneName=spine_04}}
     {targetBone {boneName=spine_05}}
     {targetBone {boneName=pelvis}}
    ]
    constraintWeights [1,1,1,1,1,1]
   }
   "Constraint - dyn_pouch_bk_MSTR_03 (PPPPPP)"!(
    ComponentPose=@n9
    Alpha=1
    ConstraintWeights_0=0.3587
    ConstraintWeights_1:float=0.252
    ConstraintWeights_2:float=0.1479
    ConstraintWeights_3:float=0.076
    ConstraintWeights_4:float=0.0013
    ConstraintWeights_5:float=0.1641
   ) -> (Pose) {
    alphaBoolBlend {blendOption=Linear}
    boneToModify {boneName=dyn_pouch_bk_MSTR_03}
    constraintSetup [
     {targetBone {boneName=spine_01}}
     {targetBone {boneName=spine_02}}
     {targetBone {boneName=spine_03}}
     {targetBone {boneName=spine_04}}
     {targetBone {boneName=spine_05}}
     {targetBone {boneName=pelvis}}
    ]
    constraintWeights [1,1,1,1,1,1]
   }
   "Transform (Modify) Bone - Bone: dyn_shoulderpad_r_base"(Alpha=1, Translation="1,-0.25,0", Rotation="3,0,0", Scale="1,1,1") {
    rotationMode=BMM_Additive
    rotationSpace=BCS_BoneSpace
    translationMode=BMM_Additive
    translationSpace=BCS_BoneSpace
    alphaBoolBlend {blendOption=Linear}
    alphaScaleBiasClamp {interpSpeedIncreasing=40}
    boneToModify {boneName=dyn_shoulderpad_r_base}
    rotation {pitch=3}
    translation=<1 -0.25>
   }
   "Transform (Modify) Bone - Bone: dyn_poncho_back_r_top_03"!(ComponentPose, bAlphaBoolEnabled:bool=Get bIsAccelerating2D(), Translation="0,0,0", Rotation="10,0,5", Scale="1,1,1") -> (Pose) {
    alphaInputType=Bool
    rotationMode=BMM_Additive
    rotationSpace=BCS_BoneSpace
    scaleMode=BMM_Additive
    translationMode=BMM_Additive
    alphaBoolBlend {blendOption=Linear}
    boneToModify {boneName=dyn_poncho_back_r_top_03}
    rotation {pitch=10, roll=5}
   }
   n10 = "Constraint - dyn_poncho_front_l_bot_02 (P)"(Alpha=1, ConstraintWeights_0=1) {
    alphaBoolBlend {blendOption=Linear}
    boneToModify {boneName=dyn_poncho_front_l_bot_02}
    constraintSetup [{targetBone {boneName=dyn_poncho_l_sleeve_front_top_01}}]
    constraintWeights [1]
   }
   n11 = "Constraint - dyn_poncho_front_l_bot_03 (P)"(ComponentPose=@n10, Alpha=1, ConstraintWeights_0=1) {
    alphaBoolBlend {blendOption=Linear}
    boneToModify {boneName=dyn_poncho_front_l_bot_03}
    constraintSetup [{targetBone {boneName=dyn_poncho_l_sleeve_front_top_02}}]
    constraintWeights [1]
   }
   n12 = "Transform (Modify) Bone - Bone: dyn_poncho_front_l_bot_02"(ComponentPose=@n11, Alpha=1, Translation="0,-4,0", Rotation="0,-5,10", Scale="1,1,1") {
    rotationMode=BMM_Additive
    translationMode=BMM_Additive
    alphaBoolBlend {blendOption=Linear}
    boneToModify {boneName=dyn_poncho_front_l_bot_02}
    rotation {roll=10, yaw=-5}
    translation {y=-4}
   }
   "Transform (Modify) Bone - Bone: dyn_poncho_front_l_bot_03"(ComponentPose=@n12, Alpha=1, Translation="0,0,0", Rotation="0,0,-7", Scale="1,1,1") {
    rotationMode=BMM_Additive
    translationMode=BMM_Additive
    alphaBoolBlend {blendOption=Linear}
    boneToModify {boneName=dyn_poncho_front_l_bot_03}
    rotation {roll=-7}
   }
   n13 = "Transform (Modify) Bone - Bone: dyn_shoulderpad_MSTR"(bAlphaBoolEnabled=Get bIsSkyDiving(), Translation="0,0,1", Rotation="0.000,0.000,60", Scale="1,1,1") {
    alphaInputType=Bool
    rotationMode=BMM_Additive
    rotationSpace=BCS_BoneSpace
    translationMode=BMM_Additive
    translationSpace=BCS_BoneSpace
    alphaBoolBlend {blendInTime=0.1, blendOption=Linear, blendOutTime=0.1}
    alphaScaleBiasClamp {interpSpeedIncreasing=40}
    boneToModify {boneName=dyn_shoulderpad_MSTR}
    rotation {roll=60}
    translation {z=1}
   }
   n14 = "Transform (Modify) Bone - Bone: dyn_shoulderpad_MSTR"(ComponentPose=@n13, bAlphaBoolEnabled=Get bIsParachuteOpened(), Translation="0,-1,1", Rotation="0.000,0.000,-20", Scale="1,1,1") {
    alphaInputType=Bool
    rotationMode=BMM_Additive
    rotationSpace=BCS_BoneSpace
    translationMode=BMM_Additive
    translationSpace=BCS_BoneSpace
    alphaBoolBlend {blendInTime=0.1, blendOption=Linear, blendOutTime=0.1}
    alphaScaleBiasClamp {interpSpeedIncreasing=40}
    boneToModify {boneName=dyn_shoulderpad_MSTR}
    rotation {roll=-20}
    translation {y=-1, z=1}
   }
   "Transform (Modify) Bone - Bone: dyn_shoulderpad_MSTR"(ComponentPose=@n14, bAlphaBoolEnabled=Get bIsSkydiveDiveMode(), Translation="0,0,0", Rotation="0,0,-30", Scale="1,1,1") {
    alphaInputType=Bool
    rotationMode=BMM_Additive
    rotationSpace=BCS_BoneSpace
    translationMode=BMM_Additive
    translationSpace=BCS_BoneSpace
    alphaBoolBlend {blendInTime=0.1, blendOption=Linear, blendOutTime=0.1}
    alphaScaleBiasClamp {interpSpeedIncreasing=40}
    boneToModify {boneName=dyn_shoulderpad_MSTR}
    rotation {roll=-30}
   }
   "Transform (Modify) Bone - Bone: dyn_shoulderpad_MSTR"!(ComponentPose, Alpha:float=1, Translation="0,0,0", Rotation="0,0,0", Scale="1,1,1") -> (Pose) {
    rotationMode=BMM_Additive
    rotationSpace=BCS_BoneSpace
    translationMode=BMM_Additive
    translationSpace=BCS_BoneSpace
    alphaBoolBlend {blendOption=Linear}
    alphaScaleBiasClamp {interpSpeedIncreasing=40}
    boneToModify {boneName=dyn_shoulderpad_MSTR}
   }
   n15 = RigidBody(ComponentPose=Local To Component(LocalPose=Use cached pose 'Pre_Rigidbody'())) {
    bClampLinearTranslationLimitToRefPose=true
    bOverrideWorldGravity=true
    overridePhysicsAsset=&/M_MED_Progressive_Jonesy_Wombat_Front
    simulationSpace=BaseBoneSpace
    alphaBoolBlend {blendOption=Linear}
    baseBoneRef {boneName=root}
    componentAppliedLinearAccClamp=<3000 3000 3000>
    componentLinearAccScale=<0.4 0.25 0.45>
    componentLinearVelScale=<0.35 0.2 0.2>
    overrideWorldGravity {z=-980}
   }
   n16 = "Transform (Modify) Bone - Bone: dyn_poncho_front_l_bot_03"(ComponentPose=@n15, Alpha=1, Translation="0,0,0", Rotation="0,0,-20", Scale="1,1,1") {
    rotationMode=BMM_Additive
    translationMode=BMM_Additive
    alphaBoolBlend {blendOption=Linear}
    boneToModify {boneName=dyn_poncho_front_l_bot_03}
    rotation {roll=-20}
   }
   n17 = "Transform (Modify) Bone - Bone: dyn_poncho_front_r_top_01"(ComponentPose=@n16, Alpha=Get frontend_pose() -> (frontend_pose:double), Translation="0,1,0", Rotation="0,0,0", Scale="1,1,1") {
    rotationMode=BMM_Additive
    translationMode=BMM_Additive
    alphaBoolBlend {blendOption=Linear}
    boneToModify {boneName=dyn_poncho_front_r_top_01}
    translation {y=1}
   }
   n18 = Component To Local(ComponentPose=@n17)
   n19 = RigidBody(ComponentPose=Local To Component(LocalPose=Use cached pose 'Pre_Rigidbody'())) {
    bClampLinearTranslationLimitToRefPose=true
    bOverrideWorldGravity=true
    overridePhysicsAsset=&/M_MED_Progressive_Jonesy_Wombat_Physics
    simulationSpace=BaseBoneSpace
    alphaBoolBlend {blendOption=Linear}
    baseBoneRef {boneName=spine_05}
    componentAppliedLinearAccClamp=<3000 3000 3000>
    componentLinearAccScale=<0.5 0.35 0.45>
    componentLinearVelScale=<0.5 0.35 0.35>
    overrideWorldGravity {z=-980}
   }
   n20 = Component To Local(ComponentPose=@n19)
   n21 = Get bIsInFrontEnd()
   n22 = "Constraint - dyn_poncho_back_l_bot_master (PPPPPPPPPPPPPPPPPPPPPPPP)"!(
    ComponentPose=Local To Component(LocalPose=Blend Poses by bool(BlendPose_0=@n18, BlendPose_1=@n20, bActiveValue=@n21) {blendType=Linear})
    Alpha=float - float(B=Get L_Arm_Back() -> (L_Arm_Back:double)) {is=title K2Node_CallFunction}
    ConstraintWeights_0=0
    ConstraintWeights_1:float=0
    ConstraintWeights_2:float=0.5829
    ConstraintWeights_3:float=0.0341
    ConstraintWeights_4:float=0
    ConstraintWeights_5:float=0.009
    ConstraintWeights_6:float=0.0017
    ConstraintWeights_7:float=0.0139
    ConstraintWeights_8:float=0.0249
    ConstraintWeights_9:float=0.0357
    ConstraintWeights_10:float=0.0375
    ConstraintWeights_11:float=0.1069
    ConstraintWeights_12:float=0.1202
    ConstraintWeights_13:float=0
    ConstraintWeights_14:float=0
    ConstraintWeights_15:float=0
    ConstraintWeights_16:float=0
    ConstraintWeights_17:float=0.0329
    ConstraintWeights_18:float=0.0003
    ConstraintWeights_19:float=0
    ConstraintWeights_20:float=0
    ConstraintWeights_21:float=0
    ConstraintWeights_22:float=0
    ConstraintWeights_23:float=0
   ) -> (Pose) {
    alphaBoolBlend {blendOption=Linear}
    boneToModify {boneName=dyn_poncho_back_l_bot_master}
    constraintSetup [
     {targetBone {boneName=dyn_poncho_front_r_top_01}}
     {targetBone {boneName=dyn_poncho_front_r_top_02}}
     {targetBone {boneName=dyn_poncho_l_sleeve_back_top_02}}
     {targetBone {boneName=upperarm_twist_02_l}}
     {targetBone {boneName=spine_02}}
     {targetBone {boneName=spine_03}}
     {targetBone {boneName=spine_04}}
     {targetBone {boneName=spine_05}}
     {targetBone {boneName=dyn_poncho_back_l_top_03}}
     {targetBone {boneName=dyn_poncho_back_l_top_02}}
     {targetBone {boneName=dyn_poncho_back_l_top_01}}
     {targetBone {boneName=clavicle_l}}
     {targetBone {boneName=upperarm_twist_01_l}}
     {targetBone {boneName=clavicle_r}}
     {targetBone {boneName=neck_02}}
     {targetBone {boneName=neck_01}}
     {targetBone {boneName=dyn_poncho_back_r_top_01}}
     {targetBone {boneName=upperarm_l}}
     {targetBone {boneName=dyn_l_sleeve_top_01}}
     {targetBone {boneName=dyn_l_sleeve_top_02}}
     {targetBone {boneName=spine_01}}
     {targetBone {boneName=dyn_poncho_l_sleeve_front_top_02}}
     {targetBone {boneName=dyn_poncho_l_sleeve_front_top_01}}
     {targetBone {boneName=lowerarm_l}}
    ]
    constraintWeights [1,1,1,1,1,1,1,1,1,1,1,1,1,1,1,1,1,1,1,1,1,1,1,1]
   }
   n23 = Get Aim_Up() -> (Aim_Up:double)
   n24 = "Transform (Modify) Bone - Bone: dyn_poncho_back_l_top_master"(ComponentPose=@n22, Alpha=@n23, Translation="5,0,2", Rotation="0,0,0", Scale="1,1,1") {
    rotationMode=BMM_Additive
    scaleMode=BMM_Additive
    translationMode=BMM_Additive
    alphaBoolBlend {blendOption=Linear}
    boneToModify {boneName=dyn_poncho_back_l_top_master}
    translation {x=5, z=2}
   }
   n25 = "Transform (Modify) Bone - Bone: dyn_poncho_r_sleeve_back_01"(ComponentPose=@n24, Alpha=@n23, Translation="-7,7,0", Rotation="-60,0,0", Scale="1,1,1") {
    rotationMode=BMM_Additive
    scaleMode=BMM_Additive
    translationMode=BMM_Additive
    alphaBoolBlend {blendOption=Linear}
    boneToModify {boneName=dyn_poncho_r_sleeve_back_01}
    rotation {pitch=-60}
    translation=<-7 7>
   }
   n26 = "Transform (Modify) Bone - Bone: dyn_poncho_r_fold_back_01"(ComponentPose=@n25, Alpha=@n23, Translation="-3,0,0", Rotation="-15,0,0", Scale="1,1,1") {
    rotationMode=BMM_Additive
    scaleMode=BMM_Additive
    translationMode=BMM_Additive
    alphaBoolBlend {blendOption=Linear}
    boneToModify {boneName=dyn_poncho_r_fold_back_01}
    rotation {pitch=-15}
    translation {x=-3}
   }
   n27 = "Transform (Modify) Bone - Bone: dyn_poncho_l_sleeve_back_top_01"!(ComponentPose=@n26, bAlphaBoolEnabled:bool=Get bIsAccelerating2D(), Translation="0,6,6", Rotation="0,0,0", Scale="1,1,1") -> (Pose) {
    alphaInputType=Bool
    rotationMode=BMM_Additive
    rotationSpace=BCS_BoneSpace
    scaleMode=BMM_Additive
    translationMode=BMM_Additive
    translationSpace=BCS_BoneSpace
    alphaBoolBlend {blendInTime=0.5, blendOption=Linear, blendOutTime=0.5}
    boneToModify {boneName=dyn_poncho_l_sleeve_back_top_01}
    translation {y=6, z=6}
   }
   n28 = float * float(A=Get L_Arm_Back(), B=Get Arm_Back_Strength() -> (Arm_Back_Strength:double))
   n29 = "Transform (Modify) Bone - Bone: dyn_poncho_l_sleeve_back_top_master"(ComponentPose=@n27, Alpha=@n28, Translation="0,-10,10", Rotation="0,0,0", Scale="1,1,1") {
    rotationMode=BMM_Additive
    rotationSpace=BCS_BoneSpace
    scaleMode=BMM_Additive
    translationMode=BMM_Additive
    alphaBoolBlend {blendOption=Linear}
    boneToModify {boneName=dyn_poncho_l_sleeve_back_top_master}
    translation {y=-10, z=10}
   }
   n30 = "Transform (Modify) Bone - Bone: dyn_poncho_back_l_top_master"(ComponentPose=@n29, Alpha=@n28, Translation="0,-2.5,0", Rotation="0,0,0", Scale="1,1,1") {
    rotationMode=BMM_Additive
    scaleMode=BMM_Additive
    translationMode=BMM_Additive
    alphaBoolBlend {blendOption=Linear}
    boneToModify {boneName=dyn_poncho_back_l_top_master}
    translation {y=-2.5}
   }
   n31 = SelectFloat(A=0.8, B=0, bPickA=Get Is_Wearing_Backbling() -> (Is_Wearing_Backbling:bool))
   n32 = "Constraint - dyn_poncho_back_l_top_02 (P)"(ComponentPose=@n30, Alpha=@n31, ConstraintWeights_0=1) {
    alphaBoolBlend {blendOption=Linear}
    boneToModify {boneName=dyn_poncho_back_l_top_02}
    constraintSetup [{targetBone {boneName=spine_03}}]
    constraintWeights [1]
   }
   n33 = "Constraint - dyn_poncho_back_r_top_02 (P)"(ComponentPose=@n32, Alpha=@n31, ConstraintWeights_0=1) {
    alphaBoolBlend {blendOption=Linear}
    boneToModify {boneName=dyn_poncho_back_r_top_02}
    constraintSetup [{targetBone {boneName=spine_03}}]
    constraintWeights [1]
   }
   n34 = "Constraint - dyn_poncho_back_l_top_03 (P)"(ComponentPose=@n33, Alpha=@n31, ConstraintWeights_0=1) {
    alphaBoolBlend {blendOption=Linear}
    boneToModify {boneName=dyn_poncho_back_l_top_03}
    constraintSetup [{targetBone {boneName=spine_03}}]
    constraintWeights [1]
   }
   n35 = "Constraint - dyn_poncho_back_r_top_03 (P)"(ComponentPose=@n34, Alpha=@n31, ConstraintWeights_0=1) {
    alphaBoolBlend {blendOption=Linear}
    boneToModify {boneName=dyn_poncho_back_r_top_03}
    constraintSetup [{targetBone {boneName=spine_03}}]
    constraintWeights [1]
   }
   n36 = "Constraint - dyn_poncho_r_fold_back_01 (P)"(ComponentPose=@n35, Alpha=@n31, ConstraintWeights_0=1) {
    alphaBoolBlend {blendOption=Linear}
    boneToModify {boneName=dyn_poncho_r_fold_back_01}
    constraintSetup [{targetBone {boneName=spine_03}}]
    constraintWeights [1]
   }
   n37 = SelectFloat(A=1, B=0, bPickA=Get bIsSkyDiving())
   n38 = "Constraint - dyn_poncho_l_sleeve_back_top_master (PP)"!(ComponentPose=@n36, Alpha=@n37, ConstraintWeights_0=1, ConstraintWeights_1:float=0) -> (Pose) {
    alphaBoolBlend {blendOption=Linear}
    boneToModify {boneName=dyn_poncho_l_sleeve_back_top_master}
    constraintSetup [{targetBone {boneName=upperarm_twist_01_l}}, {targetBone {boneName=upperarm_twist_02_l}}]
    constraintWeights [1,1]
   }
   n39 = "Transform (Modify) Bone - Bone: dyn_poncho_l_sleeve_back_top_01"!(ComponentPose=@n38, Alpha:float=@n37, Translation="0,0,0", Rotation="10,-10,0", Scale="1,1,1") -> (Pose) {
    rotationMode=BMM_Additive
    scaleMode=BMM_Additive
    translationMode=BMM_Additive
    alphaBoolBlend {blendOption=Linear}
    boneToModify {boneName=dyn_poncho_l_sleeve_back_top_01}
    rotation {pitch=10, yaw=-10}
   }
   n40 = "Transform (Modify) Bone - Bone: dyn_poncho_r_sleeve_back_01"(ComponentPose=@n39, Alpha=@n37, Translation="0,-5,12", Rotation="0,-15,30", Scale="1,1,1") {
    rotationMode=BMM_Additive
    scaleMode=BMM_Additive
    translationMode=BMM_Additive
    alphaBoolBlend {blendOption=Linear}
    boneToModify {boneName=dyn_poncho_r_sleeve_back_01}
    rotation {roll=30, yaw=-15}
    translation {y=-5, z=12}
   }
   n41 = "Transform (Modify) Bone - Bone: dyn_poncho_r_fold_back_01"(ComponentPose=@n40, Alpha=@n37, Translation="0,-5,5", Rotation="0,-10,10", Scale="1,1,1") {
    rotationMode=BMM_Additive
    scaleMode=BMM_Additive
    translationMode=BMM_Additive
    alphaBoolBlend {blendOption=Linear}
    boneToModify {boneName=dyn_poncho_r_fold_back_01}
    rotation {roll=10, yaw=-10}
    translation {y=-5, z=5}
   }
   Output Pose(Result=Component To Local(ComponentPose=@n41)) {layerGroup=None}
   n42 = "Transform (Modify) Bone - Bone: dyn_arm_blades"(
    ComponentPose=Local To Component(LocalPose=Copy Pose From Mesh() {bUseAttachedParent=true})
    Alpha=1
    Translation="0,0,0"
    Rotation="-1,0.000,0.000"
    Scale="1,1,1"
   ) {rotationMode=BMM_Additive, alphaBoolBlend {blendOption=Linear}, boneToModify {boneName=dyn_arm_blades}, rotation {pitch=-1}}
   n43 = "Constraint - dyn_pouch_lt_MSTR_01 (PPPPPP)"!(
    ComponentPose=@n42
    Alpha=1
    ConstraintWeights_0=0.3589
    ConstraintWeights_1:float=0.1969
    ConstraintWeights_2:float=0.1192
    ConstraintWeights_3:float=0.0239
    ConstraintWeights_4:float=0
    ConstraintWeights_5:float=0.301
   ) -> (Pose) {
    alphaBoolBlend {blendOption=Linear}
    boneToModify {boneName=dyn_pouch_lt_MSTR_01}
    constraintSetup [
     {targetBone {boneName=spine_01}}
     {targetBone {boneName=spine_02}}
     {targetBone {boneName=spine_03}}
     {targetBone {boneName=spine_04}}
     {targetBone {boneName=spine_05}}
     {targetBone {boneName=pelvis}}
    ]
    constraintWeights [1,1,1,1,1,1]
   }
   n44 = "Constraint - dyn_pouch_lt_MSTR_02 (PPPPPP)"!(
    ComponentPose=@n43
    Alpha=1
    ConstraintWeights_0=0.3853
    ConstraintWeights_1:float=0.1921
    ConstraintWeights_2:float=0.0999
    ConstraintWeights_3:float=0.0268
    ConstraintWeights_4:float=0
    ConstraintWeights_5:float=0.2959
   ) -> (Pose) {
    alphaBoolBlend {blendOption=Linear}
    boneToModify {boneName=dyn_pouch_lt_MSTR_02}
    constraintSetup [
     {targetBone {boneName=spine_01}}
     {targetBone {boneName=spine_02}}
     {targetBone {boneName=spine_03}}
     {targetBone {boneName=spine_04}}
     {targetBone {boneName=spine_05}}
     {targetBone {boneName=pelvis}}
    ]
    constraintWeights [1,1,1,1,1,1]
   }
   n45 = "Constraint - dyn_pouch_lt_MSTR_03 (PPPPPP)"!(
    ComponentPose=@n44
    Alpha=1
    ConstraintWeights_0=0.3969
    ConstraintWeights_1:float=0.2037
    ConstraintWeights_2:float=0.094
    ConstraintWeights_3:float=0.0437
    ConstraintWeights_4:float=0.0001
    ConstraintWeights_5:float=0.2616
   ) -> (Pose) {
    alphaBoolBlend {blendOption=Linear}
    boneToModify {boneName=dyn_pouch_lt_MSTR_03}
    constraintSetup [
     {targetBone {boneName=spine_01}}
     {targetBone {boneName=spine_02}}
     {targetBone {boneName=spine_03}}
     {targetBone {boneName=spine_04}}
     {targetBone {boneName=spine_05}}
     {targetBone {boneName=pelvis}}
    ]
    constraintWeights [1,1,1,1,1,1]
   }
   n46 = "Constraint - dyn_crossbow_01_l_MSTR (PP)"!(ComponentPose=@n45, Alpha=1, ConstraintWeights_0=0.75, ConstraintWeights_1:float=0.25) -> (Pose) {
    alphaBoolBlend {blendOption=Linear}
    boneToModify {boneName=dyn_crossbow_01_l_MSTR}
    constraintSetup [{targetBone {boneName=pelvis}}, {targetBone {boneName=thigh_l}}]
    constraintWeights [1,1]
   }
   n47 = "Constraint - dyn_shirt_01_r_MSTR (R)"(ComponentPose=@n46, Alpha=0, ConstraintWeights_0=1) {
    alphaBoolBlend {blendOption=Linear}
    boneToModify {boneName=dyn_shirt_01_r_MSTR}
    constraintSetup [{transformType=Rotation, targetBone {boneName=thigh_r}}]
    constraintWeights [1]
   }
   n48 = Get bIsCrouching()
   n49 = Get bIsTargeting()
   n50 = "Constraint - dyn_shirt_fr_01_r_MSTR (RR)"!(ComponentPose=@n47, bAlphaBoolEnabled:bool=NOT Boolean(A=OR Boolean(A=@n49, B=@n48)), ConstraintWeights_0=1, ConstraintWeights_1:float=0) -> (Pose) {
    alphaInputType=Bool
    alphaBoolBlend {blendInTime=0.1, blendOption=Linear, blendOutTime=0.1}
    boneToModify {boneName=dyn_shirt_fr_01_r_MSTR}
    constraintSetup [{transformType=Rotation, targetBone {boneName=thigh_twist_01_r}}, {transformType=Rotation, targetBone {boneName=thigh_r}}]
    constraintWeights [1,1]
   }
   n51 = "Constraint - dyn_cb_buckle_01_MSTR (PPPPPPP)"!(
    ComponentPose=@n50
    Alpha=1
    ConstraintWeights_0=0.3326
    ConstraintWeights_1:float=0.2047
    ConstraintWeights_2:float=0.1342
    ConstraintWeights_3:float=0.0277
    ConstraintWeights_4:float=0.0001
    ConstraintWeights_5:float=0.2984
    ConstraintWeights_6:float=0.0023
   ) -> (Pose) {
    alphaBoolBlend {blendOption=Linear}
    boneToModify {boneName=dyn_cb_buckle_01_MSTR}
    constraintSetup [
     {targetBone {boneName=spine_01}}
     {targetBone {boneName=spine_02}}
     {targetBone {boneName=spine_03}}
     {targetBone {boneName=spine_04}}
     {targetBone {boneName=spine_05}}
     {targetBone {boneName=pelvis}}
     {targetBone {boneName=thigh_twist_01_l}}
    ]
    constraintWeights [1,1,1,1,1,1,1]
   }
   n52 = "Transform (Modify) Bone - Bone: dyn_shirt_fr_01_r_MSTR"(ComponentPose=@n51, Alpha=1, Translation="-0.000,0,0.371", Rotation="0,0,0", Scale="1,1,1") {
    rotationMode=BMM_Additive
    rotationSpace=BCS_BoneSpace
    scaleMode=BMM_Additive
    scaleSpace=BCS_BoneSpace
    translationMode=BMM_Additive
    translationSpace=BCS_BoneSpace
    alphaBoolBlend {blendOption=Linear}
    boneToModify {boneName=dyn_shirt_fr_01_r_MSTR}
    translation {z=0.371}
   }
   n53 = "Transform (Modify) Bone - Bone: dyn_shirt_fr_01_r_MSTR"!(ComponentPose=@n52, bAlphaBoolEnabled:bool=Get bIsJumping(), Translation="0,0,0", Rotation="15,0,0", Scale="1,1,1") -> (Pose) {
    alphaInputType=Bool
    rotationMode=BMM_Additive
    rotationSpace=BCS_BoneSpace
    scaleMode=BMM_Additive
    scaleSpace=BCS_BoneSpace
    translationMode=BMM_Additive
    translationSpace=BCS_BoneSpace
    alphaBoolBlend {blendOption=Linear}
    boneToModify {boneName=dyn_shirt_fr_01_r_MSTR}
    rotation {pitch=15}
   }
   n54 = "Transform (Modify) Bone - Bone: dyn_crossbow_01_l"!(ComponentPose=@n53, bAlphaBoolEnabled:bool=@n48, Translation="-1.5,-1,-1", Rotation="0,0,0", Scale="1,1,1") -> (Pose) {
    alphaInputType=Bool
    rotationMode=BMM_Additive
    translationMode=BMM_Additive
    alphaBoolBlend {blendInTime=0.1, blendOption=Linear, blendOutTime=0.1}
    boneToModify {boneName=dyn_crossbow_01_l}
    translation=<-1.5 -1 -1>
   }
   n55 = "Transform (Modify) Bone - Bone: dyn_shirt_fr_01_r_MSTR"(ComponentPose=@n54, bAlphaBoolEnabled=@n48, Translation="-3,3,1", Rotation="0,0,0", Scale="1,1,1") {
    alphaInputType=Bool
    rotationMode=BMM_Additive
    rotationSpace=BCS_BoneSpace
    scaleMode=BMM_Additive
    scaleSpace=BCS_BoneSpace
    translationMode=BMM_Additive
    translationSpace=BCS_BoneSpace
    alphaBoolBlend {blendOption=Linear}
    boneToModify {boneName=dyn_shirt_fr_01_r_MSTR}
    translation=<-3 3 1>
   }
   n56 = "Transform (Modify) Bone - Bone: dyn_shirt_01_r_MSTR"!(ComponentPose=@n55, bAlphaBoolEnabled:bool=@n48, Translation="-3,3,1", Rotation="0,0,0", Scale="1,1,1") -> (Pose) {
    alphaInputType=Bool
    rotationMode=BMM_Additive
    rotationSpace=BCS_BoneSpace
    scaleMode=BMM_Additive
    scaleSpace=BCS_BoneSpace
    translationMode=BMM_Additive
    translationSpace=BCS_BoneSpace
    alphaBoolBlend {blendOption=Linear}
    boneToModify {boneName=dyn_shirt_01_r_MSTR}
    translation=<-3 3 1>
   }
   n57 = Get windVector() -> (windVector:Vector)
   n58 = RigidBody!(ComponentPose=@n56, Alpha, OverrideWorldGravity:Vector=@n57, ExternalForce) -> (Pose) {
    bClampLinearTranslationLimitToRefPose=true
    bOverrideWorldGravity=true
    overridePhysicsAsset=/BRCosmetics/Characters/Player/Male/Medium/Bodies/M_MED_Progressive_Jonesy_Kepler/Meshes/M_MED_Progressive_Jonesy_kepler_Cb_Physics
    simulationSpace=BaseBoneSpace
    alphaBoolBlend {blendOption=Linear}
    baseBoneRef {boneName=root}
    componentAppliedLinearAccClamp=<3000 3000 3000>
    componentLinearAccScale=<0.3 0.3 0.7>
    componentLinearVelScale=<0.3 0.3 0.7>
    overrideWorldGravity {z=-980}
   }
   n59 = "Transform (Modify) Bone - Bone: dyn_crossbow_03_l"!(ComponentPose=@n58, bAlphaBoolEnabled:bool=@n48, Translation="-0.5,-2,-0.5", Rotation="-5,0,0", Scale="1,1,1") -> (Pose) {
    alphaInputType=Bool
    translationMode=BMM_Additive
    alphaBoolBlend {blendInTime=0.1, blendOption=Linear, blendOutTime=0.1}
    boneToModify {boneName=dyn_crossbow_03_l}
    rotation {pitch=-5}
    translation=<-0.5 -2 -0.5>
   }
   n60 = RigidBody(ComponentPose=@n59, OverrideWorldGravity=@n57) {
    bClampLinearTranslationLimitToRefPose=true
    bOverrideWorldGravity=true
    overridePhysicsAsset=/BRCosmetics/Characters/Player/Male/Medium/Bodies/M_MED_Progressive_Jonesy_Kepler/Meshes/M_MED_Progressive_Jonesy_kepler_Physics
    simulationSpace=BaseBoneSpace
    alphaBoolBlend {blendOption=Linear}
    baseBoneRef {boneName=root}
    componentAppliedLinearAccClamp=<3000 3000 3000>
    componentLinearAccScale=<0.3 0.3 0.4>
    componentLinearVelScale=<0.3 0.3 0.4>
    overrideWorldGravity {z=-980}
   }
   n61 = "Constraint - dyn_poncho_front_l_bot_master (TPPP)"!(ComponentPose=@n60, Alpha=1, ConstraintWeights_0=1, ConstraintWeights_1:float=0, ConstraintWeights_2:float=0, ConstraintWeights_3:float=0) -> (Pose) {
    alphaBoolBlend {blendOption=Linear}
    boneToModify {boneName=dyn_poncho_front_l_bot_master}
    constraintSetup [
     {transformType=Translation, targetBone {boneName=dyn_poncho_l_sleeve_front_top_01}}
     {targetBone {boneName=upperarm_l}}
     {targetBone {boneName=clavicle_l}}
     {targetBone {boneName=dyn_poncho_l_sleeve_front_top_01}}
    ]
    constraintWeights [1,1,1,1]
   }
   Save cached pose 'Pre_Rigidbody'(Pose=Component To Local(ComponentPose=@n61)) {cacheName=Pre_Rigidbody}
   Get pistonAimAlpha_R() -> (pistonAimAlpha_R:double)
  ]}
 EventGraph {nodes [
   n1 = Event OnBeginCrouching()
   n2 = Delay(execute=[@n1.then, @n3.then, @n29.then], Duration=0.1)
   ResetDynamics(execute=[@n30.then, @n2.then], InTeleportType=TeleportPhysics)
   > ResetDynamics(InTeleportType=ResetPhysics)
   n3 = Event OnEndPlayingEmote() {is=title K2Node_Event}
   Event BlueprintUpdateAnimation()
   > n4 = Sequence!(execute) -> (then_0, then_1, then_2:Exec, then_3:Exec, then_4:Exec)
   n5 = Break Fort Anim Input Skydiving External Force(FortAnimInput_SkydivingExternalForce=Get SkydivingExternalForceInput()) -> (
    bUseSkydivingVectorForce=~
    SkydivingVectorForce=~
    HeadToPelvisDirection=~
    FloatingMultiplier=~
    FloatingAdditive=~
    DivingMultiplier=~
    DivingAdditive=~
    ParachutingMultiplier=~
    ParachutingAdditive=~
   )
   > Set windVector(windVector:Vector=SelectVector(A=vector * float(A=@n5.SkydivingVectorForce, B=0.15), B="0, 0,-980", bPickA=Get bIsSkyDiving())) -> (Output_Get:Vector)
   n6 = Sequence(execute=@n4.then_2)
   n7 = GetOwnerSkeletalMeshComponent()
   n8 = GetDeltaTransformFromRefPose(self=@n7, BoneName=upperarm_r, BaseName=clavicle_r)
   Set uparmFwdBk_R(execute=@n6.then_1, uparmFwdBk_R:double=BreakRotator(InRot=BreakTransform(InTransform=@n8).Rotation).Yaw) -> (Output_Get:double)
   n9 = "Equal Exactly (String)"(A=Get weapon() -> (weapon:String), B=Pistol)
   n10 = Branch(execute=@n6.then_2, Condition=@n9)
   > Set pistonAimAlpha_R(pistonAimAlpha_R:double=NormalizeToRange!(Value:double=Get uparmFwdBk_R() -> (uparmFwdBk_R:double), RangeMin:double=50, RangeMax:double=120) -> (ReturnValue:double)) -> (Output_Get:double)
   Set pistonAimAlpha_R(execute=@n10.else, pistonAimAlpha_R=0)
   n11 = Get weapon()
   n12 = Branch(execute=@n6.then_3, Condition=NOT Boolean(A="Equal Exactly (String)"(A=@n11, B=pistol)))
   > Set pistonAimAlpha_R(pistonAimAlpha_R=NormalizeToRange(Value=Get uparmFwdBk_R(), RangeMin=30, RangeMax=60))
   Set pistonAimAlpha_R(execute=@n12.else, pistonAimAlpha_R=0)
   n13 = GetAnimInstance(self=GetOwnerSkeletalMeshComponent())
   Is Valid(exec=@n6.then_4, InputObject=@n13)
   > Set frontend_pose(frontend_pose:double=SelectFloat(
     A=1
     B=0
     bPickA=Montage_IsPlaying(self=@n13, Montage=/BRCosmetics/Animation/Game/MainPlayer/Menu/BR/CustomSkins/_S16/ProgressiveJonesy/ProgressiveJonesy_Male_Idle_Montage)
    )) -> (Output_Get:double)
   n14 = GetDeltaTransformFromRefPose(self=@n7, BoneName=upperarm_l, BaseName=clavicle_l)
   Set uparmFwdBk_L(execute=@n6.then_0, uparmFwdBk_L:double=BreakRotator(InRot=BreakTransform(InTransform=@n14).Rotation).Yaw) -> (Output_Get:double)
   n15 = GetOwnerSkeletalMeshComponent()
   n16 = BreakRotator(InRot=BreakTransform(InTransform=GetDeltaTransformFromRefPose(self=@n15, BoneName=upperarm_r, BaseName=clavicle_r)).Rotation)
   Set Aim_Up(execute=@n4.then_3, Aim_Up:double=MapRangeClamped(Value=@n16.Yaw, InRangeA=70, InRangeB=130)) -> (Output_Get:double)
   > Set Aim_Up_Adjusted(Aim_Up_Adjusted:double=MapRangeClamped(Value=@n16.Yaw, InRangeA=60, InRangeB=100)) -> (Output_Get:double)
   n17 = BreakTransform(InTransform=GetDeltaTransformFromRefPose(self=GetOwnerSkeletalMeshComponent(), BoneName=upperarm_r, BaseName=clavicle_r))
   > Set Aim_Down(Aim_Down:double=MapRangeClamped(Value=BreakRotator(InRot=@n17.Rotation).Yaw, InRangeA=40, InRangeB=90)) -> (Output_Get:double)
   n18 = GetOwnerSkeletalMeshComponent()
   n19 = BreakRotator(InRot=BreakTransform(InTransform=GetDeltaTransformFromRefPose(self=@n18, BoneName=clavicle_l, BaseName=spine_05)).Rotation)
   n20 = MapRangeClamped(Value=@n19, InRangeA=0, InRangeB=10)
   > Set Rifle_Adjustment(Rifle_Adjustment:double=@n20) -> (Output_Get:double)
   > Set Rifle_Mid(Rifle_Mid:double=float - float(A=MapRangeClamped(Value=@n19, InRangeA=0, InRangeB=1), B=@n20)) -> (Output_Get:double)
   n21 = BreakTransform(InTransform=GetDeltaTransformFromRefPose(self=GetOwnerSkeletalMeshComponent(), BoneName=upperarm_l, BaseName=clavicle_l))
   > Set L_Arm_Back(L_Arm_Back:double=MapRangeClamped(Value=BreakRotator(InRot=@n21.Rotation).Yaw, InRangeA=0, InRangeB=-10)) -> (Output_Get:double)
   n22 = BreakTransform(InTransform=GetDeltaTransformFromRefPose(self=GetOwnerSkeletalMeshComponent(), BoneName=upperarm_l, BaseName=clavicle_l))
   > Set L_Arm_Down(L_Arm_Down:double=MapRangeClamped(Value=BreakRotator(InRot=@n22.Rotation).Yaw, InRangeA=13, InRangeB=3)) -> (Output_Get:double)
   > n23 = Cast To FortPlayerAnimInstance(Object=GetAnimInstance(self=GetOwnerSkeletalMeshComponent()))
   > Set Arm_Back_Strength(Arm_Back_Strength:double=0) -> (Output_Get:double)
   > n24 = Switch on EFortWeaponCoreAnimation!(execute, Selection:EFortWeaponCoreAnimation Enum=Get WeaponCoreAnim(self=@n23)) -> (
    Melee:Exec
    Pistol:Exec
    Shotgun:Exec
    PaperBlueprint:Exec
    Rifle:Exec
    MeleeOneHand:Exec
    MachinePistol:Exec
    RocketLauncher:Exec
    GrenadeLauncher:Exec
    GoingCommando:Exec
    AssaultRifle:Exec
    TacticalShotgun:Exec
    SniperRifle:Exec
    TrapPlacement:Exec
    ShoulderLauncher:Exec
    AbilityDecoTool:Exec
    Crossbow:Exec
    C4:Exec
    RemoteControl:Exec
    DualWield:Exec
    AR_BullPup:Exec
    AR_ForwardGrip:Exec
    MedPackPaddles:Exec
    SMG_P90:Exec
    AR_DrumGun:Exec
    Consumable_Small:Exec
    Consumable_Large:Exec
    Balloon:Exec
    MountedTurret:Exec
    CreativeTool:Exec
    ExplosiveBow:Exec
    AshtonIndigo:Exec
    AshtonChicago:Exec
    MeleeDualWield:Exec
    Unarmed:Exec
   )
   n25 = Branch(execute=[@n24.MeleeDualWield, @n24.MeleeOneHand], Condition=Get bIsMoving2D())
   > Set Arm_Back_Strength(Arm_Back_Strength=0.6)
   Set Arm_Back_Strength(execute=@n25.else, Arm_Back_Strength=0.4)
   n26 = Branch(execute=@n24.Pistol, Condition=Get bIsTargeting())
   > Set Arm_Back_Strength(Arm_Back_Strength=0)
   Set Arm_Back_Strength(execute=@n26.else, Arm_Back_Strength=1)
   n27 = Cast To FortPlayerAnimInstance(execute=@n4.then_1, Object=GetAnimInstance(self=GetOwnerSkeletalMeshComponent()))
   > Set weapon(weapon:String=Enum to String!(Enumerator:EFortWeaponCoreAnimation Enum=Get WeaponCoreAnim(self=@n27)) -> (ReturnValue:String)) -> (Output_Get:String)
   Event OnCharacterCustomizationLoaded()
   n28 = CheckForBackbling()
   > Set Is_Wearing_Backbling(Is_Wearing_Backbling:bool=OR Boolean(A=@n28, B=@n28.bIsBackblingLong)) -> (Output_Get:bool)
   > ResetDynamics(InTeleportType=TeleportPhysics)
   > ResetDynamics(InTeleportType=ResetPhysics)
   n29 = Event OnEndSkydiving()
   Event BlueprintInitializeAnimation()
   > n30 = Delay(Duration=0.15)
   n31 = Sequence!(execute) -> (then_0, then_1, then_2)
   n32 = Sequence!(execute=@n31.then_2) -> (then_0, then_1, then_2, then_3:Exec)
   n33 = GetOwnerSkeletalMeshComponent()
   n34 = GetDeltaTransformFromRefPose(self=@n33, BoneName=upperarm_r, BaseName=clavicle_r)
   Set uparmFwdBk_R(execute=@n32.then_1, uparmFwdBk_R=BreakRotator(InRot=BreakTransform(InTransform=@n34).Rotation).Yaw)
   n35 = "Equal Exactly (String)"(A=Get weapon(), B=Pistol)
   n36 = Branch(execute=@n32.then_2, Condition=@n35)
   > Set pistonAimAlpha_R(pistonAimAlpha_R=NormalizeToRange(Value=Get uparmFwdBk_R(), RangeMin=50, RangeMax=120))
   Set pistonAimAlpha_R(execute=@n36.else, pistonAimAlpha_R=0)
   n37 = Get weapon()
   n38 = Branch(execute=@n32.then_3, Condition=NOT Boolean(A="Equal Exactly (String)"(A=@n37, B=pistol)))
   > Set pistonAimAlpha_R(pistonAimAlpha_R=NormalizeToRange(Value=Get uparmFwdBk_R(), RangeMin=30, RangeMax=60))
   Set pistonAimAlpha_R(execute=@n38.else, pistonAimAlpha_R=0)
   n39 = GetDeltaTransformFromRefPose(self=@n33, BoneName=upperarm_l, BaseName=clavicle_l)
   Set uparmFwdBk_L(execute=@n32.then_0, uparmFwdBk_L=BreakRotator(InRot=BreakTransform(InTransform=@n39).Rotation).Yaw)
   n40 = Cast To FortPlayerAnimInstance(execute=@n31.then_1, Object=GetAnimInstance(self=GetOwnerSkeletalMeshComponent()))
   > Set weapon(weapon=Enum to String(Enumerator=Get WeaponCoreAnim(self=@n40)))
   n41 = Get SkydivingExternalForceInput()
   SelectVector(
    A=vector * float(
     A=Break Fort Anim Input Skydiving External Force(FortAnimInput_SkydivingExternalForce=@n41) -> (
      bUseSkydivingVectorForce=~
      SkydivingVectorForce=~
      HeadToPelvisDirection=~
      FloatingMultiplier=~
      FloatingAdditive=~
      DivingMultiplier=~
      DivingAdditive=~
      ParachutingMultiplier=~
      ParachutingAdditive=~
     ).SkydivingVectorForce
     B=0.15
    )
    B="0, 0,-980"
    bPickA=Get bIsSkyDiving()
   )
   AND Boolean(A=NOT Boolean(A=@n35), B=NOT Boolean(A=Get bIsSkyDiving()))
   AND Boolean(A=NOT Boolean(A=@n9), B=NOT Boolean(A=Get bIsSkyDiving()))
   AND Boolean(A=false, B=false)
   AND Boolean(A=false, B=false)
   Get bIsSkyDiving()
   Get bIsSkyDiving()
  ]}
 NewMacro_0 {nodes [
   Inputs!()
   Outputs()
  ]}
]
variables [
 Aim_Down {type=number}
 Aim_Up {type=number}
 Aim_Up_Adjusted {type=number}
 arm rotation_l {type=number}
 Arm_Back_Strength {type=number}
 frontend_pose {type=number}
 Is_Wearing_Backbling {type=boolean}
 L_Arm_Back {type=number}
 L_Arm_Down {type=number}
 pistonAimAlpha_R {type=number}
 Rifle_Adjustment {type=number}
 Rifle_Mid {type=number}
 uparmFwdBk_L {type=number}
 uparmFwdBk_R {type=number}
 weapon {type=string, default=""}
 windVector {type=Vector}
]
