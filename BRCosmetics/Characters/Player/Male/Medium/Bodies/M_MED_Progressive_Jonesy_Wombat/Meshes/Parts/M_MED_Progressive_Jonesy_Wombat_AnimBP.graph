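# Post-style AnimBP for the Progressive Jonesy Wombat male body part, adding secondary motion to its hood bones.
# It copies the pose from the attached parent mesh, then runs spring-driven, heavily damped Anim Dynamics on the right, left and top hood bones in series.
# Weighted multi-bone constraints on the left and right hood bones precede the simulation but have zero alpha, so contribute nothing.
# The event graph has no logic.
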AnimBlueprint /BRCosmetics/Characters/Player/Male/Medium/Bodies/M_MED_Progressive_Jonesy_Wombat/Meshes/Parts/M_MED_Progressive_Jonesy_Wombat_AnimBP
target_skeleton=&/M_MED_Progressive_Jonesy_Wombat_Skeleton
graphs [
 AnimGraph {nodes [
   n1 = "Constraint - dyn_hood_r (PPPPPPP)"!(
    ComponentPose=Local To Component(LocalPose=Copy Pose From Mesh() {bUseAttachedParent=true})
    Alpha=0
    ConstraintWeights_0=0.0303
    ConstraintWeights_1:float=0.0104
    ConstraintWeights_2:float=0.0111
    ConstraintWeights_3:float=0.3382
    ConstraintWeights_4:float=0.3305
    ConstraintWeights_5:float=0
    ConstraintWeights_6:float=0.2795
   ) -> (Pose) {
    alphaBoolBlend {blendOption=Linear}
    boneToModify {boneName=dyn_hood_r}
    constraintSetup [
     {targetBone {boneName=head}}
     {targetBone {boneName=clavicle_l}}
     {targetBone {boneName=dyn_hood_r}}
     {targetBone {boneName=spine_05}}
     {targetBone {boneName=clavicle_r}}
     {targetBone {boneName=dyn_hood_l}}
     {targetBone {boneName=neck_01}}
    ]
    constraintWeights [1,1,1,1,1,1,1]
   }
   n2 = "Constraint - dyn_hood_l (PPPPP)"!(
    ComponentPose=@n1
    Alpha=0
    ConstraintWeights_0=0.0319
    ConstraintWeights_1:float=0.0442
    ConstraintWeights_2:float=0.2077
    ConstraintWeights_3:float=0.5498
    ConstraintWeights_4:float=0.1664
   ) -> (Pose) {
    alphaBoolBlend {blendOption=Linear}
    boneToModify {boneName=dyn_hood_l}
    constraintSetup [
     {targetBone {boneName=head}}
     {targetBone {boneName=neck_01}}
     {targetBone {boneName=head}}
     {targetBone {boneName=spine_05}}
     {targetBone {boneName=clavicle_l}}
    ]
    constraintWeights [1,1,1,1,1]
   }
   n3 = Anim Dynamics - Bone: dyn_hood_r(ComponentPose=@n2, Alpha=0.9, ExternalForce="0,0,0") {
    angularDampingOverride=0.999
    angularSpringConstant=1000
    bAngularSpring=true
    bGravityOverrideInSimSpace=true
    bLinearSpring=true
    bOverrideAngularDamping=true
    bOverrideLinearDamping=true
    bUseGravityOverride=true
    gravityScale=0.3
    linearDampingOverride=0.999
    linearSpringConstant=200
    alphaBoolBlend {blendOption=Linear}
    boundBone {boneName=dyn_hood_r}
    componentAppliedLinearAccClamp=<1000 1000 1000>
    componentLinearAccScale=<1 1 1>
    componentLinearVelScale=<1 1 1>
    gravityOverride {z=-700}
    physicsBodyDefinitions [{
      sphereCollisionRadius=0
      boundBone {boneName=dyn_hood_r}
      boxExtents=<5 10 10>
      constraintSetup {
       angularTargetAxis=AxisY
       angularLimitsMax=<0.5 0.5 5>
       angularLimitsMin=<-0.5 -0.5 -5>
       angularTarget=<0 0 0>
       linearAxesMax=<0.7 0.7 0.7>
       linearAxesMin=<-0.7 -0.7 -0.7>
      }
      localJointOffset=<0 1 0>
     }]
   }
   n4 = Anim Dynamics - Bone: dyn_hood_l(ComponentPose=@n3, Alpha=1, ExternalForce="0,0,0") {
    angularDampingOverride=0.999
    angularSpringConstant=1000
    bAngularSpring=true
    bGravityOverrideInSimSpace=true
    bLinearSpring=true
    bOverrideAngularDamping=true
    bOverrideLinearDamping=true
    bUseGravityOverride=true
    gravityScale=0.3
    linearDampingOverride=0.999
    linearSpringConstant=200
    alphaBoolBlend {blendOption=Linear}
    boundBone {boneName=dyn_hood_l}
    componentAppliedLinearAccClamp=<1000 1000 1000>
    componentLinearAccScale=<1 1 1>
    componentLinearVelScale=<1 1 1>
    gravityOverride {z=-700}
    physicsBodyDefinitions [{
      sphereCollisionRadius=0
      boundBone {boneName=dyn_hood_l}
      boxExtents=<5 10 10>
      constraintSetup {
       angularTargetAxis=AxisY
       angularLimitsMax=<0.5 0.5 5>
       angularLimitsMin=<-0.5 -0.5 -5>
       angularTarget=<0 0 0>
       linearAxesMax=<0.7 0.7 0.7>
       linearAxesMin=<-0.7 -0.7 -0.7>
      }
      localJointOffset=<0 1 0>
     }]
   }
   n5 = Anim Dynamics - Bone: dyn_hood_top(ComponentPose=@n4, Alpha=0.9, ExternalForce="0,0,0") {
    angularDampingOverride=0.999
    angularSpringConstant=1000
    bAngularSpring=true
    bGravityOverrideInSimSpace=true
    bLinearSpring=true
    bOverrideAngularDamping=true
    bOverrideLinearDamping=true
    bUseGravityOverride=true
    gravityScale=0.3
    linearDampingOverride=0.999
    linearSpringConstant=200
    alphaBoolBlend {blendOption=Linear}
    boundBone {boneName=dyn_hood_top}
    componentAppliedLinearAccClamp=<1000 1000 1000>
    componentLinearAccScale=<1 1 1>
    componentLinearVelScale=<1 1 1>
    gravityOverride {z=-700}
    physicsBodyDefinitions [{
      sphereCollisionRadius=0
      boundBone {boneName=dyn_hood_top}
      boxExtents=<5 10 10>
      constraintSetup {
       angularTargetAxis=AxisY
       angularLimitsMax=<0.5 0.5 5>
       angularLimitsMin=<-0.5 -0.5 -5>
       angularTarget=<0 0 0>
       linearAxesMax=<0.7 0.4 0.7>
       linearAxesMin=<-0.7 -0.4 -0.7>
      }
      localJointOffset=<0 1 0>
     }]
   }
   Output Pose(Result=Component To Local(ComponentPose=@n5)) {layerGroup=None}
  ]}
 EventGraph {nodes [Event BlueprintUpdateAnimation()]}
]
variables []
